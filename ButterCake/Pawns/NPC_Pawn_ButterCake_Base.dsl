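; Base pawn Blueprint for the ButterCake NPC creature, handling its berserk mood, sneezing after vacuuming pickups, and players riding its back.
; On the server, damage fills a tag-weighted gauge that triggers berserk, whose retriggerable timer waits for any prevent-stop tag before ending.
; Riders get a quest GameplayEffect and, during berserk, provoke a ground-pound reflex.
; Clients handle variant materials, anim input events, alert audio, and effect culling by budget and death.
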(event Custom|OnFootStep
  (bind _ns_buttercake_footstep_alt (Variables|Default|GetNS_Buttercake_Footstep_Alt))
  (Components|Activation|Activate _ns_buttercake_footstep_alt true)
  (Niagara|SetNiagaraVariable(Int32) _ns_buttercake_footstep_alt "User.FootBone" (Variables|Default|GetFootBone)))

(event EventBeginPlay
  ;[1 1 1] Init
  (|CallParentFunction)
  (Rendering|SetVisibility (Variables|Default|GetNS_Buttercake_Drool_Ambient) true false)
  (AI|SetMinSignificanceforAnimationCurves 0.0)
  (Components|DestroyComponent (Variables|Default|GetNPCStatusWidget))
  (Networking|SwitchHasAuthority
    (:Authority
      (Sneeze|ComputeSneezeThreshold)
      (Class|FortNavModifierComponent|SetComponentAffectingNavMesh (Variables|Default|GetFortNavModifier_For_Avoidance) (Variables|Character|GetCapsuleComponent))
      (Ability|Tasks|WaitGameplayTagRemoveFromActor 0 (Variables|Default|GetGameplayTagPreventStopBerserk)
        (:then
          (Interaction|BindEventtoOnReaction (Variables|Default|GetFortAthenaCollisionReaction_Bottom) (delegate Custom|OnPushActor))
          ;[1 1 1] Unstuck Behavior
          (ButterCakeUnstuck|BindEventtoOnBlocked (Variables|Default|GetButterCakeUnstuck) (delegate Custom|OnBlocked))
          ;[1 1 1] Check at regular interval if players are on BC Back during Berserk Mode
          (Utilities|Time|SetTimerbyEvent (delegate Custom|CheckPlayerOnBack) (ScalableFloat|GetValueatLevel (Variables|Default|GetGroundPoundReflexIntervalCheck)) true)
          ;[1 1 1] Perf Optim: Deactive Drool during some GAs
          (if (ScalableFloat|GetValueasBool (Variables|Default|GetCanDeactivateDrool))
            (Ability|Tasks|WaitGameplayTagAddToActor 0 (Variables|Default|GetDeactivateDroolTag)
              (:then
                (Ability|Tasks|WaitGameplayTagRemoveFromActor 0 (Variables|Default|GetDeactivateDroolTag)
                  (:Removed
                    (Variables|Default|SetIsDroolEnable true))))
              (:Added
                (Variables|Default|SetIsDroolEnable false)))))
        (:Removed
          (CallFunction|OnPreventStopBerserkRemoved))))
    (:Remote
      (Mood|UpdateAlertStateAudio))))

(event Custom|OnPushActor
  (params (HitActor :type Actor%20Object%20Reference) (bFromAsyncSweepBox :type Boolean) (CollisionReactionType :type FCollisionReactionType%20Enum :default "Nothing") (bIsFirstContinuousReactionOnDelayableActor :type Boolean) (HitResult :type Hit%20Result%20Structure%20%28by%20ref%29))
  ;[1 1 1] Init
  (Utilities|Casting|CastToFortPawn HitActor)
  (GameplayCue|ExecuteGameplayCue HitActor (Variables|Default|GetGameplayCueTagPushTarget)))

(event Custom|OnBlocked
  ;[1 1 1] Unstuck Behavior
  (Tags|CancelAbilitiesWithTags (Variables|GameplayAbilities|GetAIPawnAbilitySystemComponent) (Variables|Default|GetCancelAbilitiesTags)))

(event Custom|CheckPlayerOnBack
  ;[1 1 1] Check at regular interval if players are on BC Back during Berserk Mode
  (if (Variables|Default|GetIsBerserk)
    (bind _isaplayer (CallFunction|IsAPlayeronMyBack))
    (if _isaplayer
      (CallFunction|TryPlayGroundPoundReflex))))

(event EventEndPlay
  (params (EndPlayReason :type EEndPlayReason%20Enum :default "Destroyed"))
  ;[1 1 1] End Play
  (|CallParentFunction EndPlayReason)
  (Interaction|UnbindallEventsfromOnReaction (Variables|Default|GetFortAthenaCollisionReaction_Bottom)))

(event FortAIPawn|EventOnPawnStartedUsingMeAsWalkableBase
  (params (PawnUsingThisAsBase :type Pawn%20Object%20Reference))
  ;[1 1 1] Pawn jumping on BC Back
  (Networking|SwitchHasAuthority
    (:Authority
      (CallFunction|ApplyOrRemoveRidingButterCakeGE PawnUsingThisAsBase true)
      (if (Variables|Default|GetIsBerserk)
        (CallFunction|TryPlayGroundPoundReflex)))))

(event Custom|TryPlayGroundPoundReflex
  ;[1 1 1] Pawn jumping on BC Back
  (Abilities|TryActivateAbilitiesbyTag (Ability|GetAbilitySystemComponent self) (Variables|Default|GetGroundPoundReflexTag) true))

(event FortAIPawn|EventOnPawnStoppedUsingMeAsWalkableBase
  (params (PawnUsingThisAsBase :type Pawn%20Object%20Reference))
  ;[1 1 1] Pawn jumping off BC back
  (Networking|SwitchHasAuthority
    (:Authority
      (CallFunction|ApplyOrRemoveRidingButterCakeGE PawnUsingThisAsBase))))

(event EventOnDamageServer
  (params (Damage :type Float%20%28single-precision%29) (DamageTags :type Gameplay%20Tag%20Container%20Structure%20%28by%20ref%29) (Momentum :type Vector :default "0, 0, 0") (HitInfo :type Hit%20Result%20Structure%20%28by%20ref%29) (InstigatedBy :type Controller%20Object%20Reference) (DamageCauser :type Actor%20Object%20Reference) (EffectContext :type Gameplay%20Effect%20Context%20Handle%20Structure))
  ;[1 1 1] Manage Berserk State (SERVER)
  (|CallParentFunction Damage DamageTags Momentum HitInfo InstigatedBy DamageCauser EffectContext)
  (if (Variables|Default|GetIsBerserk)
    (CallFunction|TickBerserk false InstigatedBy)
    (else
      (bind _mustbeberserk (Mood|UpdateBerserkGauge DamageTags))
      (CallFunction|TickBerserk _mustbeberserk InstigatedBy))))

(event Custom|TickBerserk
  (params (bForceBerserk :type Boolean) (InstigatedBy :type Controller%20Object%20Reference) (bForceBerserkEnd :type Boolean))
  ;[1 1 1] Manage Berserk State (SERVER)
  (if (not (Variables|Default|GetIsBerserk))
    (Utilities|FlowControl|Sequence
      (:then_0
        (Utilities|FlowControl|RetriggerableDelay (ScalableFloat|GetValueatLevel (Variables|Default|GetBerserkGaugeResetTimer)))
        (Variables|Default|SetBerserkGauge 0.0))
      (:then_1
        (if bForceBerserk
          (GameplayEffects|ApplyGameplayEffectToSelf :self (Variables|GameplayAbilities|GetAIPawnAbilitySystemComponent) :GameplayEffectClass "/ButterCake/Abilities/GE_ButterCake_Berserk.GE_ButterCake_Berserk_C" :Level 0.0)
          (Mood|OnBerserkBegin InstigatedBy)))))
  (Variables|Default|SetIsBerserkTimerEnded false)
  (if bForceBerserkEnd
    (Variables|Default|SetLastBerserkDuration 0.0)
    (else
      (bind _berserkduration (Mood|GetBerserkDuration))
      (Variables|Default|SetLastBerserkDuration _berserkduration)))
  (Utilities|FlowControl|RetriggerableDelay (Variables|Default|GetLastBerserkDuration))
  (CallFunction|OnBerserkTimerEnd))

(event Custom|On%20PreventStopBerserk%20Removed
  ;[1 1 1] Manage Berserk State (SERVER)
  (if (Variables|Default|GetIsBerserkTimerEnded)
    (GameplayEffects|RemoveActiveEffectswithTags (Variables|GameplayAbilities|GetAIPawnAbilitySystemComponent) (Variables|Default|GetGameplayTagBerserk))
    (Default|CallOnBerserkEndDelegate)))

(event Custom|OnBerserkTimerEnd
  ;[1 1 1] Manage Berserk State (SERVER)
  (Variables|Default|SetIsBerserkTimerEnded true)
  (if (not (GameplayTags|HasMatchingGameplayTag (Variables|GameplayAbilities|GetAIPawnAbilitySystemComponent) (Variables|Default|GetGameplayTagPreventStopBerserk)))
    (GameplayEffects|RemoveActiveEffectswithTags (Variables|GameplayAbilities|GetAIPawnAbilitySystemComponent) (Variables|Default|GetGameplayTagBerserk))
    (Default|CallOnBerserkEndDelegate)))

(event EventOnDeathPlayEffects
  (params (Damage :type Float%20%28single-precision%29) (DamageTags :type Gameplay%20Tag%20Container%20Structure%20%28by%20ref%29) (Momentum :type Vector :default "0, 0, 0") (HitInfo :type Hit%20Result%20Structure%20%28by%20ref%29) (InstigatedBy :type Fort%20Pawn%20Object%20Reference) (DamageCauser :type Actor%20Object%20Reference) (EffectContext :type Gameplay%20Effect%20Context%20Handle%20Structure))
  ;[1 1 1] Disable effects and animation (which will stop audio from playing) on death (on the client)
  (|CallParentFunction Damage DamageTags Momentum HitInfo InstigatedBy DamageCauser EffectContext)
  (Rendering|SetVisibility (Variables|Default|GetNS_Buttercake_Drool_Ambient) false false)
  (Rendering|SetVisibility (Variables|Default|GetWildlife_FootDust) false false)
  (Rendering|SetVisibility (Variables|Default|GetNS_Buttercake_Footstep_Alt) false false)
  (Components|SkeletalMesh|SetAnimInstanceClass (Variables|Character|GetMesh) 0))

(event EventOnApplyEffectOptimization
  (params (NewBudgetValue :type Integer))
  (bind _returnvalue (<= NewBudgetValue 1))
  ;[1 1 1] Disable effects on client when LOD is low enough (>=2)
  (|CallParentFunction NewBudgetValue)
  (if (not (FortPawn|IsDead))
    (Rendering|SetVisibility (Variables|Default|GetNS_Buttercake_Footstep_Alt) _returnvalue true)
    (Rendering|SetVisibility (Variables|Default|GetNS_Buttercake_Drool_Ambient) _returnvalue true)))

(fn ApplyorRemoveRidingButterCakeGE
  (params (PawnUsingThisAsBase :type Pawn%20Object%20Reference) (bIsOnWalkableBase :type Boolean))
  (bind _returnvalue (Fortnite|GetFortPlayerControllerfromActor PawnUsingThisAsBase))
  (bind _returnvalue_2 (Ability|GetFortAbilitySystemComponent _returnvalue))
  ;[1 1 1] If Pawn is riding on ButterCake, apply a GE that grants a tag that can be pulled for quests; if not riding, remove the tag
  (if (Utilities|IsValid _returnvalue_2 :!class /Script/Engine.KismetSystemLibrary)
    (if bIsOnWalkableBase
      (GameplayEffects|ApplyGameplayEffectToTarget :self _returnvalue_2 :GameplayEffectClass (LocalVariables|GetGE_QuestRidingButterCake) :Target _returnvalue_2 :Level 0.0)
      (return)
      (else
        (GameplayEffects|RemoveActiveEffectswithGrantedTags _returnvalue_2 (LocalVariables|GetGrantedQuestTag))
        (return)))))

(fn ApplyVariation
  (Networking|SwitchHasAuthority
    (:Remote
      (Mood|ChangeMaterial "/Game/Packages/Fortress_SharedMaterials/Base_Material/CharacterDeRezShader.CharacterDeRezShader")
      (Mood|InitializeVariantMaterials (Variables|Default|GetVariantIndex))
      (ConsoleVars|ValidateInsetShadows)
      (bind _returnvalue (Class|FortAnimInputEventBool|CreateAnimInputEventBool "aie.ButterCake.Berserk" (Variables|Default|GetIsBerserk)))
      (Animation|TriggerAnimInputEvent _returnvalue)
      (bind _returnvalue_2 (Class|FortAnimInputEventBool|CreateAnimInputEventBool "aie.ButterCake.Lured" (Variables|Default|GetIsLured)))
      (Animation|TriggerAnimInputEvent _returnvalue_2))))

(fn BlueprintModifyIncomingDamage
  (params (Damage :type Float%20%28single-precision%29) (DamageTags :type Gameplay%20Tag%20Container%20Structure%20%28by%20ref%29) (EffectContext :type Gameplay%20Effect%20Context%20Handle%20Structure%20%28by%20ref%29) (EventInstigator :type Controller%20Object%20Reference) (DamageCauser :type Actor%20Object%20Reference))
  (bind _returnvalue (Utilities|GetFrameCount))
  (|CallParentFunction Damage DamageTags EffectContext EventInstigator DamageCauser)
  (LocalVariables|SetNewDamageBC 0.0)
  (if (not (Math|Boolean|ANDBoolean (== (Variables|Damage|GetLastDamageInstigatorBC) EventInstigator) (== DamageCauser (Variables|Damage|GetLast%20DamageCauserBC)) (== _returnvalue (Variables|Damage|GetLastFrameCountBC))))
    (LocalVariables|SetNewDamageBC 1.0))
  (Variables|Damage|SetLastDamageInstigatorBC EventInstigator)
  (Variables|Damage|SetLast%20DamageCauserBC DamageCauser)
  (Variables|Damage|SetLastFrameCountBC _returnvalue)
  (return (LocalVariables|GetNewDamageBC)))

(fn CanSneeze
  (return (Math|Boolean|ORBoolean (>= (Variables|Sneeze|GetPickedUpItemsSinceLastSneeze) (Variables|Sneeze|GetRequiredItemsToSneeze)) (>= (Math|Conversions|ToInteger(Byte) (Variables|Default|GetCurrentMaxRarityItemsVacuumed)) (ScalableFloat|GetValueasInteger (LocalVariables|GetRarityThresholdToSneeze))) (Variables|Sneeze|GetbForceSneezeOnceWhenPossible))))

(fn ChangeMaterial
  (params (NewMaterial :type Material%20Interface%20Object%20Reference))
  (bind _returnvalue (Customization|GetSkeletalMeshforPartType (Variables|Default|GetFortCharacterParts) "Body" :!class /Script/FortniteGame.FortCharacterPartsComponent))
  (Utilities|IsValid _returnvalue
    (:Is%20Valid
      (if (Variables|Default|GetIsBerserk)
        (Rendering|Material|SetScalarParameterValueonMaterials _returnvalue "GameTimeFromBP_Berzerk" (Utilities|Time|GetGameTimeinSeconds))
        (Rendering|Material|SetScalarParameterValueonMaterials _returnvalue "GameTimeFromBP_Lured" 9.999999843067494e+17)
        (Rendering|Material|SetScalarParameterValueonMaterials _returnvalue "GameTimeFromBP_Neutral" 9.999999843067494e+17)
        (elif (Variables|Default|GetIsLured)
          _
          (else
            (Rendering|Material|SetScalarParameterValueonMaterials _returnvalue "GameTimeFromBP_Neutral" (Utilities|Time|GetGameTimeinSeconds))))))))

(fn ComputeSneezeThreshold
  (Variables|Sneeze|SetRequiredItemsToSneeze (Math|Random|RandomIntegerinRange (Math|Float|Truncate (ScalableFloat|GetValueatLevel (Variables|Sneeze|GetRequiredItemsToSneezeMin))) (Math|Float|Truncate (ScalableFloat|GetValueatLevel (Variables|Sneeze|GetRequiredItemsToSneezeMax))))))

(fn ForceBerserkBegin
  (params (Instigator :type Controller%20Object%20Reference))
  (if (not (Variables|Default|GetIsBerserk))
    (CallFunction|TickBerserk true Instigator)))

(fn ForceBerserkEnd
  (params (Instigator :type Controller%20Object%20Reference))
  (if (Variables|Default|GetIsBerserk)
    (CallFunction|TickBerserk false Instigator true)))

(fn ForceSneezeOnceWhenPossible
  (Variables|Sneeze|SetbForceSneezeOnceWhenPossible true))

(fn GetBackpackLocation
  (return (Transformation|GetSocketLocation (Actor|GetComponentbyClass "/Script/FortniteGame.FortPooledSkelMeshComponent") (LocalVariables|GetInsideMouthBone))))

(fn GetBerserkDuration
  (if (> (Variables|Configuration|GetBerserkDurationOverridden) 0)
    (return (Variables|Configuration|GetBerserkDurationOverridden))
    (else
      (return (ScalableFloat|GetValueatLevel (Variables|Default|GetBerserkDuration))))))

(fn GetBerserkValue
  (params (DamageTags :type Gameplay%20Tag%20Container%20Structure))
  (bind _berserkgaugedatatable (LocalVariables|GetBerserkGaugeDataTable))
  (bind _outrownames (DataTable|GetDataTableRowNames _berserkgaugedatatable))
  (for _array_element _outrownames
    (bind _returnvalue (Utilities|GetDataTableRow _berserkgaugedatatable _array_element))
    (if (GameplayTags|HasAnyTags DamageTags (Utilities|Struct|BreakBPBerserkGaugeStruct _returnvalue))
      (return (.Value_5_A92A2E3D4B3966268CB205984F7704F6 (Utilities|Struct|BreakBPBerserkGaugeStruct _returnvalue)))))
  ;[1 1 1] Default
  (return (ScalableFloat|GetValueatLevel (LocalVariables|GetDefaultValue))))

(fn InitializeVariantMaterials
  (params (VariantIndex :type Integer))
  (bind _returnvalue (Customization|GetSkeletalMeshforPartType (Variables|Default|GetFortCharacterParts) "Body" :!class /Script/FortniteGame.FortCharacterPartsComponent))
  (switch int VariantIndex
    (:1
      (Rendering|Material|SetMaterial _returnvalue 1 "/ButterCake/Materials/MI_Buttercake_Tongue_V2.MI_Buttercake_Tongue_V2")
      (Rendering|Material|SetMaterial _returnvalue 4 "/ButterCake/Materials/MI_Buttercake_Body_V2.MI_Buttercake_Body_V2")
      (Rendering|Material|SetMaterial _returnvalue 5 "/ButterCake/Materials/MI_Buttercake_Head_V2.MI_Buttercake_Head_V2"))
    (:2
      (Rendering|Material|SetMaterial _returnvalue 1 "/ButterCake/Materials/MI_Buttercake_Tongue_V3.MI_Buttercake_Tongue_V3")
      (Rendering|Material|SetMaterial _returnvalue 4 "/ButterCake/Materials/MI_Buttercake_Body_V3.MI_Buttercake_Body_V3")
      (Rendering|Material|SetMaterial _returnvalue 5 "/ButterCake/Materials/MI_Buttercake_Head_V3.MI_Buttercake_Head_V3"))
    (:3
      (Rendering|Material|SetMaterial _returnvalue 1 "/ButterCake/Materials/MI_Buttercake_Tongue_V4.MI_Buttercake_Tongue_V4")
      (Rendering|Material|SetMaterial _returnvalue 4 "/ButterCake/Materials/MI_Buttercake_Body_V4.MI_Buttercake_Body_V4")
      (Rendering|Material|SetMaterial _returnvalue 5 "/ButterCake/Materials/MI_Buttercake_Head_V4.MI_Buttercake_Head_V4"))
    (:4
      (Rendering|Material|SetMaterial _returnvalue 1 "/ButterCake/Materials/MI_Buttercake_Tongue_Q.MI_Buttercake_Tongue_Q")
      (Rendering|Material|SetMaterial _returnvalue 4 "/ButterCake/Materials/MI_Buttercake_Body_Q.MI_Buttercake_Body_Q")
      (Rendering|Material|SetMaterial _returnvalue 5 "/ButterCake/Materials/MI_Buttercake_Head_Q.MI_Buttercake_Head_Q"))))

(fn IsAPlayerOnMyBack
  (LocalVariables|SetPlayerFound false)
  (bind _returnvalue (FortAIPawn|FindValidPawnsUsedorUsingThisasABase self))
  (bind _array_element (Utilities|Array|ForEachLoopwithBreak _returnvalue)
    (:LoopBody
      (bind _asfort_pawn (Utilities|Casting|CastToFortPawn _array_element))
      (if (< (Math|Vector|Distance(Vector) (Transformation|GetActorLocation _asfort_pawn) (Transformation|GetActorLocation self)) (ScalableFloat|GetValueatLevel (LocalVariables|GetPlayerEjectMaxDistance)))
        (LocalVariables|SetPlayerFound true)))
    (:Completed
      _))
  (return (LocalVariables|GetPlayerFound)))

(fn OnBerserkBegin
  (params (Instigator :type Controller%20Object%20Reference))
  (Default|CallOnBerserkBeginDelegate)
  (bind _asbutter_cake_controller (Utilities|Casting|CastToButterCake_Controller (Pawn|GetController self)))
  (ButterCake-Telemetry|OnEnterBerserk (Variables|/ButterCake/Controllers/ButterCake_Controller.ButterCake_Controller_C|GetFortButterCakeComponent_Telemetry _asbutter_cake_controller) (Variables|Actor|GetInstigator)))

(fn OnRep_IsBerserk
  (bind _returnvalue (Class|FortAnimInputEventBool|CreateAnimInputEventBool "aie.ButterCake.Berserk" (Variables|Default|GetIsBerserk)))
  (Animation|TriggerAnimInputEvent _returnvalue)
  (Mood|UpdateAlertStateAudio))

(fn OnRep_IsDroolEnable
  (Networking|SwitchHasAuthority
    (:Remote
      (Components|Activation|SetActive (Variables|Default|GetNS_Buttercake_Drool_Ambient) (Variables|Default|GetIsDroolEnable) false))))

(fn OnRep_IsLured
  (bind _returnvalue (Class|FortAnimInputEventBool|CreateAnimInputEventBool "aie.ButterCake.Lured" (Variables|Default|GetIsLured)))
  (Animation|TriggerAnimInputEvent _returnvalue))

(fn OnSneeze
  ;[1 1 1] Reset
  (Variables|Sneeze|SetbForceSneezeOnceWhenPossible false)
  (Variables|Sneeze|SetPickedUpItemsSinceLastSneeze 0)
  (Variables|Default|SetCurrentMaxRarityItemsVacuumed "Common")
  (Sneeze|ComputeSneezeThreshold)
  (Default|CallOnSneezedDelegate))

(fn OnVacuumPickup
  (params (PickupsCount :type Integer) (PickupsRarityMax :type EFortRarity%20Enum :default "Common"))
  (Variables|Sneeze|SetPickedUpItemsSinceLastSneeze (+ PickupsCount (Variables|Sneeze|GetPickedUpItemsSinceLastSneeze)))
  (if (> (Math|Conversions|ToInteger(Byte) (Variables|GetPickupsRarityMax)) (Math|Conversions|ToInteger(Byte) (Variables|Default|GetCurrentMaxRarityItemsVacuumed)))
    (Variables|Default|SetCurrentMaxRarityItemsVacuumed (Variables|GetPickupsRarityMax)))
  (Default|CallOnVacuumPickupsDelegate (Variables|GetPickupsCount)))

(fn UpdateAlertStateAudio
  (bind _npc_alertleveltotagmap (Variables|NPCAlertLevel|GetNPC%20AlertLevelToTagMap))
  (bind (_value _returnvalue) (Utilities|Map|Find _npc_alertleveltotagmap "NewEnumerator1"))
  (bind _npc_alertleveltotagmap_2 (Variables|NPCAlertLevel|GetNPC%20AlertLevelToTagMap))
  (bind (_value_2 _returnvalue_2) (Utilities|Map|Find _npc_alertleveltotagmap_2 "NewEnumerator3"))
  (bind _returnvalue_3 (FortGameplayDataTrackerComponent|GetFortGameplayDataTrackerComponentManager))
  (Utilities|IsValid _returnvalue_3
    (:Is%20Valid
      (if (or _returnvalue _returnvalue_2)
        (Class|FortGameplayDataTrackerComponentManager|SetTrackedActorState _returnvalue_3 self (select (Variables|Default|GetIsBerserk) _value_2 _value))))))

(fn UpdateBerserkGauge
  (params (DamageTags :type Gameplay%20Tag%20Container%20Structure))
  (if (Variables|Configuration|GetbCanBeBerserk)
    (bind _value (Mood|GetBerserkValue DamageTags))
    (bind _output_get (Variables|Default|SetBerserkGauge (+ (* _value (Variables|Configuration|GetBerserkGaugeMultiplier)) (Variables|Default|GetBerserkGauge))))
    (if (>= _output_get (ScalableFloat|GetValueatLevel (LocalVariables|GetThreshold)))
      (Variables|Default|SetBerserkGauge 0.0)
      (return true)
      (else
        (return false)))
    (else
      (return false))))

(fn ConstructionScript
  (|CallParentFunction))

(fn validateInsetShadows
  (bind _returnvalue (Customization|GetSkeletalMeshforPartType (Variables|Default|GetFortCharacterParts) "Body" :!class /Script/FortniteGame.FortCharacterPartsComponent))
  (Utilities|IsValid _returnvalue
    (:Is%20Valid
      (if (Class|PlatformHelpersLibrary|IsPcPerfMode)
        (Rendering|SetCastInsetShadow _returnvalue false)
        (elif (Fortnite|IsMobilePlatform)
          (Rendering|SetCastInsetShadow _returnvalue false)
          (else
            (Rendering|SetCastInsetShadow _returnvalue true)))))))
